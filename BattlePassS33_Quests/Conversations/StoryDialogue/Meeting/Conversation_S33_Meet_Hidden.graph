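# Conversation database for NPC story dialogue in the Season 33 "meeting hidden" quest, giving lore about magical resonance on the Island.
# Three tagged entry points each lead to a two-line Speech chain with a shared opening line and a different closing revelation.
# Each opening line requires that quest to be active.
# Each closing "Continue" choice carries a numbered quest gameplay tag matching its entry point.
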ConversationDatabase /BattlePassS33_Quests/Conversations/StoryDialogue/Meeting/Conversation_S33_Meet_Hidden
entry_points [
 {tag=Conversation.Entry.Season33.Story.Meet.Hidden.01, destinations [#1]}
 {tag=Conversation.Entry.Season33.Story.Meet.Hidden.02, destinations [#2]}
 {tag=Conversation.Entry.Season33.Story.Meet.Hidden.03, destinations [#3]}
]
nodes [
 ConversationEntryPointNode_2 {
  id=#3
  class=ConversationEntryPointNode
  outputs [#4]
  properties {entryTag {tagName=Conversation.Entry.Season33.Story.Meet.Hidden.03}}
 }
 Speech_C_7 {
  id=#4
  class=Speech_C
  outputs [#5]
  properties {softTaskIcon=&T_UI_NPC_question, generalConfig {defaultMessage=There is ancient magical resonance here.}}
  sub_nodes [
   ConversationChoiceNode_8 {class=ConversationChoiceNode, properties {defaultChoiceDisplayText=Dark Magic is dangerous}}
   FortConversationRequirement_ControllerRequirement_4 {
    class=FortConversationRequirement_ControllerRequirement
    properties {
     nodeName=Require: Is Quest Active
     requirement {
      class=FortControllerRequirement_IsQuestActive
      properties {questDefinition=/BattlePassS33_Quests/Items/QuestItems/Story/S33_Story_Meeting/quest_S33_story_meeting_hidden}
     }
    }
   }
  ]
 }
 Speech_C_10 {
  id=#5
  class=Speech_C
  outputs []
  properties {generalConfig {defaultMessage="It echoes across time and space, into the Spirit Realm."}}
  sub_nodes [ConversationChoiceNode_11 {
    class=ConversationChoiceNode
    properties {defaultChoiceDisplayText=Continue, choiceTags {gameplayTags [{tagName=Athena.Quests.S33.Conversation.Story.Meet.Hidden.03}]}}
   }]
 }
 ConversationEntryPointNode_1 {
  id=#2
  class=ConversationEntryPointNode
  outputs [#6]
  properties {entryTag {tagName=Conversation.Entry.Season33.Story.Meet.Hidden.02}}
 }
 Speech_C_2 {
  id=#6
  class=Speech_C
  outputs [#7]
  properties {softTaskIcon=&T_UI_NPC_question, generalConfig {defaultMessage=There is ancient magical resonance here.}}
  sub_nodes [
   ConversationChoiceNode_4 {class=ConversationChoiceNode, properties {defaultChoiceDisplayText=Dark Magic is dangerous}}
   FortConversationRequirement_ControllerRequirement_1 {
    class=FortConversationRequirement_ControllerRequirement
    properties {
     nodeName=Require: Is Quest Active
     requirement {
      class=FortControllerRequirement_IsQuestActive
      properties {questDefinition=&quest_S33_story_meeting_hidden}
     }
    }
   }
  ]
 }
 Speech_C_6 {
  id=#7
  class=Speech_C
  outputs []
  properties {generalConfig {defaultMessage="It is connected, somehow, to the demons and the rage that drives them."}}
  sub_nodes [ConversationChoiceNode_7 {
    class=ConversationChoiceNode
    properties {defaultChoiceDisplayText=Continue, choiceTags {gameplayTags [{tagName=Athena.Quests.S33.Conversation.Story.Meet.Hidden.02}]}}
   }]
 }
 ConversationEntryPointNode_0 {
  id=#1
  class=ConversationEntryPointNode
  outputs [#8]
  properties {entryTag {tagName=Conversation.Entry.Season33.Story.Meet.Hidden.01}}
 }
 Speech_C_3 {
  id=#8
  class=Speech_C
  outputs [#9]
  properties {softTaskIcon=&T_UI_NPC_question, generalConfig {defaultMessage=There is ancient magical resonance here.}}
  sub_nodes [
   ConversationChoiceNode_0 {class=ConversationChoiceNode, properties {defaultChoiceDisplayText=Dark Magic is dangerous}}
   FortConversationRequirement_ControllerRequirement_0 {
    class=FortConversationRequirement_ControllerRequirement
    properties {
     nodeName=Require: Is Quest Active
     requirement {
      class=FortControllerRequirement_IsQuestActive
      properties {questDefinition=&quest_S33_story_meeting_hidden}
     }
    }
   }
  ]
 }
 Speech_C_1 {
  id=#9
  class=Speech_C
  outputs []
  properties {generalConfig {defaultMessage=Something very old and very powerful is stirring the Island's magics.}}
  sub_nodes [ConversationChoiceNode_3 {
    class=ConversationChoiceNode
    properties {defaultChoiceDisplayText=Continue, choiceTags {gameplayTags [{tagName=Athena.Quests.S33.Conversation.Story.Meet.Hidden.01}]}}
   }]
 }
]
