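# Season 33 story conversation in which an NPC explains the Island's awakened magic and demons from Spirit Realm portals, then asks the player for help.
# The entry tag leads to a linear chain of four speech nodes, advanced by Continue choices.
# The first is a main-questline choice with a quest icon, gated on the first Meeting quest being active.
# The last node's choice carries a story quest gameplay tag, marking the dialogue's end.
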ConversationDatabase /BattlePassS33_Quests/Conversations/StoryDialogue/Meeting/Conversation_S33_Meet_Q01
entry_points [{tag=Conversation.Entry.Season33.Story.Meet.01, destinations [#1]}]
nodes [
 ConversationEntryPointNode_0 {
  id=#1
  class=ConversationEntryPointNode
  outputs [#2]
  properties {entryTag {tagName=Conversation.Entry.Season33.Story.Meet.01}}
 }
 Speech_C_3 {
  id=#2
  class=Speech_C
  outputs [#3]
  properties {
   softTaskIcon=&T_UI_NPC_MainQuest
   generalConfig {defaultMessage="Ever since that mysterious object flew across the sky, this Island has changed."}
  }
  sub_nodes [
   ConversationChoiceNode_5 {
    class=ConversationChoiceNode
    properties {defaultChoiceDisplayText=Understand the nature of Demons, choiceTags {gameplayTags [{tagName=Conversation.Choice.Category.MainQuestline}]}}
   }
   FortConversationRequirement_ControllerRequirement_4 {
    class=FortConversationRequirement_ControllerRequirement
    properties {
     nodeName=Require: Is Quest Active
     requirement {
      class=FortControllerRequirement_IsQuestActive
      properties {questDefinition=/BattlePassS33_Quests/Items/QuestItems/Story/S33_Story_Meeting/quest_S33_story_meeting_q01}
     }
    }
   }
  ]
 }
 Speech_C_28 {
  id=#3
  class=Speech_C
  outputs [#4]
  properties {generalConfig {defaultMessage="Magic, like we used to tell stories about, has awoken."}}
  sub_nodes [ConversationChoiceNode_6 {class=ConversationChoiceNode, properties {defaultChoiceDisplayText=Continue}}]
 }
 Speech_C_0 {
  id=#4
  class=Speech_C
  outputs [#5]
  properties {generalConfig {defaultMessage=The portals to the Spirit Realm that we always thought were dusty relics... Demons now enter our world through them.}}
  sub_nodes [ConversationChoiceNode_7 {class=ConversationChoiceNode, properties {defaultChoiceDisplayText=Continue}}]
 }
 Speech_C_1 {
  id=#5
  class=Speech_C
  outputs []
  properties {generalConfig {defaultMessage="I need to understand why they're here, and what they want. Will you help me?"}}
  sub_nodes [ConversationChoiceNode_8 {
    class=ConversationChoiceNode
    properties {defaultChoiceDisplayText=Continue, choiceTags {gameplayTags [{tagName=Athena.Quests.S33.Conversation.Story.Meet.01}]}}
   }]
 }
]
