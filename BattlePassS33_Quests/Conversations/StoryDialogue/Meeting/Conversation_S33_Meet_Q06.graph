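# Conversation database for the Season 33 story Meeting quest 06, where the player hands collected Essence to Hope.
# A tagged entry point leads into a linear chain of four speech nodes about an artifact, awakening Island magic and invading demons.
# The first node carries the main quest icon and choice category, and requires the quest to be active.
# The final Continue choice carries the quest's conversation tag.
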ConversationDatabase /BattlePassS33_Quests/Conversations/StoryDialogue/Meeting/Conversation_S33_Meet_Q06
entry_points [{tag=Conversation.Entry.Season33.Story.Meet.06, destinations [#1]}]
nodes [
 ConversationEntryPointNode_0 {
  id=#1
  class=ConversationEntryPointNode
  outputs [#2]
  properties {entryTag {tagName=Conversation.Entry.Season33.Story.Meet.06}}
 }
 Speech_C_3 {
  id=#2
  class=Speech_C
  outputs [#3]
  properties {
   softTaskIcon=&T_UI_NPC_MainQuest
   generalConfig {defaultMessage=I came here searching for something. A strange and powerful artifact...}
  }
  sub_nodes [
   ConversationChoiceNode_5 {
    class=ConversationChoiceNode
    properties {defaultChoiceDisplayText=Give Hope the collected Essence, choiceTags {gameplayTags [{tagName=Conversation.Choice.Category.MainQuestline}]}}
   }
   FortConversationRequirement_ControllerRequirement_4 {
    class=FortConversationRequirement_ControllerRequirement
    properties {
     nodeName=Require: Is Quest Active
     requirement {
      class=FortControllerRequirement_IsQuestActive
      properties {questDefinition=/BattlePassS33_Quests/Items/QuestItems/Story/S33_Story_Meeting/quest_S33_story_meeting_q06}
     }
    }
   }
  ]
 }
 Speech_C_28 {
  id=#3
  class=Speech_C
  outputs [#4]
  properties {generalConfig {defaultMessage="Let's just say, it doesn't surprise me that this Island's myths and ancient magics are waking up."}}
  sub_nodes [ConversationChoiceNode_6 {class=ConversationChoiceNode, properties {defaultChoiceDisplayText=Continue}}]
 }
 Speech_C_0 {
  id=#4
  class=Speech_C
  outputs [#5]
  properties {generalConfig {defaultMessage=I suspect these invading demons are connected. This essence will help me and my... associate figure out how.}}
  sub_nodes [ConversationChoiceNode_7 {class=ConversationChoiceNode, properties {defaultChoiceDisplayText=Continue}}]
 }
 Speech_C_1 {
  id=#5
  class=Speech_C
  outputs []
  properties {generalConfig {defaultMessage=Thanks for your help. I'll be in touch.}}
  sub_nodes [ConversationChoiceNode_8 {
    class=ConversationChoiceNode
    properties {defaultChoiceDisplayText=Continue, choiceTags {gameplayTags [{tagName=Athena.Quests.S33.Conversation.Story.Meet.06}]}}
   }]
 }
]
